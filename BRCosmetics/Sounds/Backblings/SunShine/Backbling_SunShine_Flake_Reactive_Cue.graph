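# Sound cue for the reactive sound of the SunShine Flake back bling cosmetic.
# A Random node picks one of three equally weighted React wave variations, which a Modulator gives slight pitch and volume variation.
# It uses close-range pet attenuation, the Backblings sound class and the shared reactive back bling audio property sheet.
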SoundCue /BRCosmetics/Sounds/Backblings/SunShine/Backbling_SunShine_Flake_Reactive_Cue
root=Modulator_0
nodes [
 {id=Modulator_0, children [Random_0], properties {pitchMax=1.02, pitchMin=0.98, volumeMax=1.02, volumeMin=0.98}}
 {id=Random_0, children [WavePlayer_3,WavePlayer_4,WavePlayer_5], properties {weights [1,1,1]}}
 {id=WavePlayer_3, properties {soundWaveAssetPtr=&/Backpack_SunShine_Flake_React_01}}
 {id=WavePlayer_4, properties {soundWaveAssetPtr=&/Backpack_SunShine_Flake_React_02}}
 {id=WavePlayer_5, properties {soundWaveAssetPtr=&/Backpack_SunShine_Flake_React_03}}
]
properties {
 attenuationSettings=&Fort_Pets_Close_Athena_Att
 audioPropertiesSheet=/BRCosmetics/Sounds/Settings/PropertySheets/Reactive/APS_Cosmetic_Reactive_Backbling
 maxDistance=1800.0001
 soundClassObject=/Game/Sounds/FortSoundClassesAndModes/Fort_Backblings
 volumeMultiplier=0.95
}
